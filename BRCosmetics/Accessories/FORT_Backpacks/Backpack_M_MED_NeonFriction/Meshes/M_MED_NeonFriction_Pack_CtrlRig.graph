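# Control Rig for the NeonFriction backpack, animating its pack, door and doll with robe, hair and ribbon.
# Construction imports skeleton bones, tags them with shape, color, offset and control metadata by name, then spawns matching controls.
# Forward solve copies control transforms onto bones and applies per-group visibility from bool channels hosted on all controls.
# Backward solve repeats that control-to-bone copy instead of driving controls from bones.
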ControlRigBlueprint /BRCosmetics/Accessories/FORT_Backpacks/Backpack_M_MED_NeonFriction/Meshes/M_MED_NeonFriction_Pack_CtrlRig
graphs [
 New Function_ContainedGraph {nodes [
   n1 = RigVMFunctionEntryNode()
   RigVMFunction_Sequence = &RigVMFunction_Sequence(ExecuteContext=@n1.ExecuteContext)
   RigVMFunctionReturnNode(ExecuteContext=@RigVMFunction_Sequence.B)
   Set_Metadata_3 = RigVMDispatchNode(
    Name=Color
    NameSpace=Self
    Value=RigVMVariableNode() {variable=Color}.Value
    Item=RigVMVariableNode() {variable=Bone}.Value
    ExecuteContext=@RigVMFunction_Sequence.A
   ) {template="DISPATCH_RigDispatch_SetMetadata(in Item,in Name,in NameSpace,in Value,out Success)"}
   > Set_Metadata_1 = RigVMDispatchNode(Name=Shape, NameSpace=Self, Value=RigVMVariableNode() {variable=Shape}.Value, Item=RigVMVariableNode() {variable=Bone}.Value) {template="DISPATCH_RigDispatch_SetMetadata(in Item,in Name,in NameSpace,in Value,out Success)"}
   n2 = RigVMVariableNode() {variable=Bone}
   > Set_Metadata = RigVMDispatchNode(Name=Control, NameSpace=Self, Value=(Type=Control,Name="None"), Item=@n2.Value, Value.Name=@n2.Value.Name) {template="DISPATCH_RigDispatch_SetMetadata(in Item,in Name,in NameSpace,in Value,out Success)"}
   > Set_Metadata_2 = RigVMDispatchNode(Name=Parent, NameSpace=Self, Value=RigVMVariableNode() {variable=Parent}.Value, Item=RigVMVariableNode() {variable=Bone}.Value) {template="DISPATCH_RigDispatch_SetMetadata(in Item,in Name,in NameSpace,in Value,out Success)"}
   > Set_Metadata_4 = RigVMDispatchNode(Name=Offset, NameSpace=Self, Item=RigVMVariableNode() {variable=Bone}.Value, Value=RigVMVariableNode() {variable=Offset}.Value) {template="DISPATCH_RigDispatch_SetMetadata(in Item,in Name,in NameSpace,in Value,out Success)"}
  ]}
 New Function_ContainedGraph {nodes [
   n1 = RigVMFunctionEntryNode()
   RigVMFunction_Sequence = &RigVMFunction_Sequence(ExecuteContext=@n1.ExecuteContext)
   n2 = &RigVMFunction_Sequence(ExecuteContext=@RigVMFunction_Sequence.A)
   For_Each = RigVMDispatchNode(Array=RigVMVariableNode() {variable=Bone Array}.Value, ExecuteContext=@n2.A) {template="DISPATCH_RigVMDispatch_ArrayIterator(in Array,out Element,out Index,out Count,out Ratio)"}
   > RigVMVariableNode(Value=@For_Each.Element) {variable=local element}
   n3 = RigVMVariableNode() {variable=local element}
   Get_Metadata = RigVMDispatchNode(Default=(Rotation=(X=0,Y=-0.7071,Z=0,W=0.7071),Translation=(X=0,Y=0,Z=0),Scale3D=(X=1,Y=1,Z=1)), Name=Offset, NameSpace=Self, Item=@n3.Value) {template="DISPATCH_RigDispatch_GetMetadata(in Item,in Name,in NameSpace,in Default,out Value,out Found)"}
   Get_Metadata_1 = RigVMDispatchNode(Default=(R=0.766,G=0,B=0.7378,A=1), Name=Color, NameSpace=Self, Item=@n3.Value) {template="DISPATCH_RigDispatch_GetMetadata(in Item,in Name,in NameSpace,in Default,out Value,out Found)"}
   Get_Metadata_2 = RigVMDispatchNode(Default=Circle_Thick, Name=Shape, NameSpace=Self, Item=@n3.Value) {template="DISPATCH_RigDispatch_GetMetadata(in Item,in Name,in NameSpace,in Default,out Value,out Found)"}
   > SpawnControl = /Script/ControlRig.RigUnit_HierarchyAddControlTransform(
    InitialValue=(Rotation=(X=0,Y=0,Z=0,W=1),Translation=(X=0,Y=0,Z=0),Scale3D=(X=1,Y=1,Z=1))
    OffsetSpace=GlobalSpace
    Parent=(Type=Control,Name="None")
    Settings="(InitialSpace=LocalSpace,bUsePreferredRotationOrder=False,PreferredRotationOrder=YZX,Limits=(LimitTranslationX=(bMinimum=False,bMaximum=False),LimitTranslationY=(bMinimum=False,bMaximum=False),LimitTranslationZ=(bMinimum=False,bMaximum=False),LimitPitch=(b ...(truncated)"
    OffsetTransform=&RigUnit_GetTransform(bInitial=true, Item=@n3.Value).Transform
    Settings.Shape.Transform=@Get_Metadata.Value
    Settings.Shape.Color=@Get_Metadata_1.Value
    Settings.Shape.Name=@Get_Metadata_2.Value
    Parent.Name=/Script/ControlRig.RigUnit_HierarchyGetParent(Child=(Type=Bone,Name="None"), bDefaultParent=false, Child.Name=@For_Each.Element.Name) {is=struct RigVMUnitNode}.Parent.Name
    Name=RigVMVariableNode() {variable=local element}.Value.Name
   ) {
    is=struct RigVMUnitNode
    template="SpawnControl::Execute(in InitialValue,in Settings,in OffsetTransform,in OffsetSpace,in Parent,in Name,out Item)"
   }
   > RigVMDispatchNode(Element=@SpawnControl.Item, Array=RigVMVariableNode() {variable=ctrl array}.Value) {template="DISPATCH_RigVMDispatch_ArrayAdd(io Array,in Element,out Index)"}
   RigVMFunction_Sequence_2 = RigVMAggregateNode(ExecuteContext=@n2.B)
   ItemArray = RigVMDispatchNode(Value="((Type=Control,Name=\"dyn_doll_pelvis\"),(Type=Control,Name=\"dyn_doll_torso\"),(Type=Control,Name=\"dyn_doll_head\"),(Type=Control,Name=\"doll_upperarm_l\"),(Type=Control,Name=\"doll_lowerarm_l\"),(Type=Control,Name=\"doll_upperarm_r\"),(Type=Control,Name=\"doll_lower ...(truncated)") {template="DISPATCH_RigVMDispatch_Constant(io Value)"}
   RigVMVariableNode(ExecuteContext=@RigVMFunction_Sequence_2.B, Value=@ItemArray.Value) {variable=Doll Body Controls}
   n4 = RigVMVariableNode() {variable=ctrl array}
   Make_Array_3 = RigVMDispatchNode(
    Values=((Type=None,Name="None"),((),Type=None,Name="None"))
    Values.0=RigVMDispatchNode(Index=0, Array=@n4.Value) {template="DISPATCH_RigVMDispatch_ArrayGetAtIndex(in Array,in Index,out Element)"}.Element
    Values.1=RigVMDispatchNode(Index=1, Array=@n4.Value) {template="DISPATCH_RigVMDispatch_ArrayGetAtIndex(in Array,in Index,out Element)"}.Element
   ) {template="DISPATCH_RigVMDispatch_ArrayMake(in Values,out Array)"}
   RigVMVariableNode(Value=@Make_Array_3.Array, ExecuteContext=@RigVMFunction_Sequence_2.A) {variable=Root Controls}
   For_Each_3 = RigVMDispatchNode(Array=RigVMVariableNode() {variable=ctrl array}.Value, ExecuteContext=@RigVMFunction_Sequence_2.C) {template="DISPATCH_RigVMDispatch_ArrayIterator(in Array,out Element,out Index,out Count,out Ratio)"}
   Contains = /Script/RigVM.RigVMFunction_Contains(Search=robe, Name=@For_Each_3.Element.Name) {is=struct RigVMUnitNode, template="Contains::Execute(in Name,in Search,out Result)"}
   RigVMFunction_ControlFlowBranch = &RigVMFunction_ControlFlowBranch(Condition=@Contains.Result, ExecuteContext=@For_Each_3.ExecuteContext)
   RigVMDispatchNode(ExecuteContext=@RigVMFunction_ControlFlowBranch.True, Array=RigVMVariableNode() {variable=Robe Controls}.Value, Element=@For_Each_3.Element) {template="DISPATCH_RigVMDispatch_ArrayAdd(io Array,in Element,out Index)"}
   For_Each_1 = RigVMDispatchNode(Array=RigVMVariableNode() {variable=ctrl array}.Value, ExecuteContext=@RigVMFunction_Sequence_2.D) {template="DISPATCH_RigVMDispatch_ArrayIterator(in Array,out Element,out Index,out Count,out Ratio)"}
   n5 = &RigVMFunction_ControlFlowBranch(
    Condition=&RigVMFunction_Contains(Search=hair, Name=@For_Each_1.Element.Name) {template="Contains::Execute(in Name,in Search,out Result)"}.Result
    ExecuteContext=@For_Each_1.ExecuteContext
   )
   n6 = RigVMDispatchNode(ExecuteContext=@n5.True, Array=RigVMVariableNode() {variable=Hair Controls}.Value, Element=@For_Each_1.Element) {template="DISPATCH_RigVMDispatch_ArrayAdd(io Array,in Element,out Index)"}
   ItemArray_1 = RigVMDispatchNode(Value=((Type=Control,Name="dyn_doll_ribbon"))) {template="DISPATCH_RigVMDispatch_Constant(io Value)"}
   RigVMDispatchNode(Array=@n6.Array, ExecuteContext=@For_Each_1.Completed, Other=@ItemArray_1.Value) {template="DISPATCH_RigVMDispatch_ArrayAppend(io Array,in Other)"}
   RigVMFunctionReturnNode(ExecuteContext=@RigVMFunction_Sequence.B)
  ]}
 New Function_ContainedGraph {nodes [
   n1 = RigVMFunctionEntryNode()
   RigVMFunction_Sequence = &RigVMFunction_Sequence(ExecuteContext=@n1.ExecuteContext)
   n2 = &RigVMFunction_Sequence(ExecuteContext=@RigVMFunction_Sequence.A)
   n3 = /Script/ControlRig.RigUnit_HierarchyAddAnimationChannelBool(
    InitialValue=true
    LimitsEnabled=()
    MaximumValue=True
    MinimumValue=False
    Name=Root Visibility
    Parent=(Type=Control,Name="control rig options")
    ExecutePin=@n2.A
   ) {
    is=struct RigVMUnitNode
    template="SpawnAnimationChannel::Execute(in InitialValue,in MinimumValue,in MaximumValue,in LimitsEnabled,in Parent,in Name,out Item)"
   }
   > n4 = &RigUnit_HierarchyAddAnimationChannelBool(
    InitialValue=true
    LimitsEnabled=()
    MaximumValue=True
    MinimumValue=False
    Name=Doll Root Visibility
    Parent=(Type=Control,Name="control rig options")
   ) {template="SpawnAnimationChannel::Execute(in InitialValue,in MinimumValue,in MaximumValue,in LimitsEnabled,in Parent,in Name,out Item)"}
   > n5 = &RigUnit_HierarchyAddAnimationChannelBool(
    InitialValue=true
    LimitsEnabled=()
    MaximumValue=True
    MinimumValue=False
    Name=Doll Body Controls Visibility
    Parent=(Type=Control,Name="control rig options")
   ) {template="SpawnAnimationChannel::Execute(in InitialValue,in MinimumValue,in MaximumValue,in LimitsEnabled,in Parent,in Name,out Item)"}
   Make_Array_3 = RigVMDispatchNode(
    Values=((Type=None,Name="None"),((),Type=None,Name="None"),((),Type=None,Name="None"))
    Values.0=@n3.Item
    Values.1=@n4.Item
    Values.2=@n5.Item
   ) {template="DISPATCH_RigVMDispatch_ArrayMake(in Values,out Array)"}
   > RigVMDispatchNode(Other=@Make_Array_3.Array, Array=RigVMVariableNode() {variable=Channels array}.Value) {template="DISPATCH_RigVMDispatch_ArrayAppend(io Array,in Other)"}
   n6 = &RigUnit_HierarchyAddAnimationChannelBool(
    InitialValue=true
    LimitsEnabled=()
    MaximumValue=True
    MinimumValue=False
    Name=Robe Controls Visibility
    Parent=(Type=Control,Name="control rig options")
    ExecutePin=@n2.B
   ) {template="SpawnAnimationChannel::Execute(in InitialValue,in MinimumValue,in MaximumValue,in LimitsEnabled,in Parent,in Name,out Item)"}
   > n7 = &RigUnit_HierarchyAddAnimationChannelBool(
    InitialValue=true
    LimitsEnabled=()
    MaximumValue=True
    MinimumValue=False
    Name=Hair Controls Visibility
    Parent=(Type=Control,Name="control rig options")
   ) {template="SpawnAnimationChannel::Execute(in InitialValue,in MinimumValue,in MaximumValue,in LimitsEnabled,in Parent,in Name,out Item)"}
   > n8 = &RigUnit_HierarchyAddAnimationChannelBool(
    InitialValue=true
    LimitsEnabled=()
    MaximumValue=True
    MinimumValue=False
    Name=Door Control Visibility
    Parent=(Type=Control,Name="control rig options")
   ) {template="SpawnAnimationChannel::Execute(in InitialValue,in MinimumValue,in MaximumValue,in LimitsEnabled,in Parent,in Name,out Item)"}
   Make_Array_1 = RigVMDispatchNode(
    Values=((Type=None,Name="None"),((),Type=None,Name="None"),((),Type=None,Name="None"))
    Values.0=@n6.Item
    Values.1=@n7.Item
    Values.2=@n8.Item
   ) {template="DISPATCH_RigVMDispatch_ArrayMake(in Values,out Array)"}
   > RigVMDispatchNode(Other=@Make_Array_1.Array, Array=RigVMVariableNode() {variable=Channels array}.Value) {template="DISPATCH_RigVMDispatch_ArrayAppend(io Array,in Other)"}
   RigVMFunctionReturnNode(ExecuteContext=@RigVMFunction_Sequence.B)
  ]}
 New Function_ContainedGraph {nodes [
   n1 = RigVMFunctionEntryNode()
   RigVMFunction_Sequence = &RigVMFunction_Sequence(ExecuteContext=@n1.ExecuteContext)
   For_Each_7 = RigVMDispatchNode(ExecuteContext=@RigVMFunction_Sequence.A, Array=RigVMVariableNode() {variable=Channels array}.Value) {template="DISPATCH_RigVMDispatch_ArrayIterator(in Array,out Element,out Index,out Count,out Ratio)"}
   > /Script/ControlRig.RigUnit_SetChannelHosts(Channel=@For_Each_7.Element, Hosts=RigVMVariableNode() {variable=ctrl array}.Value) {is=struct RigVMUnitNode}
   RigVMFunctionReturnNode(ExecuteContext=@RigVMFunction_Sequence.B)
  ]}
 New Function_ContainedGraph {nodes [
   n1 = RigVMFunctionEntryNode()
   RigVMFunction_Sequence = &RigVMFunction_Sequence(ExecuteContext=@n1.ExecuteContext)
   RigVMFunction_Sequence_3 = RigVMAggregateNode(ExecuteContext=@RigVMFunction_Sequence.A)
   For_Each_3 = RigVMDispatchNode(Array=RigVMVariableNode() {variable=Hair Controls}.Value, ExecuteContext=@RigVMFunction_Sequence_3.E) {template="DISPATCH_RigVMDispatch_ArrayIterator(in Array,out Element,out Index,out Count,out Ratio)"}
   > /Script/ControlRig.RigUnit_SetControlVisibility(
    Item=@For_Each_3.Element
    bVisible=/Script/ControlRig.RigUnit_GetBoolAnimationChannel(Channel=Hair Controls Visibility, Control=control rig options, bInitial=false) {is=struct RigVMUnitNode, template="GetAnimationChannel::Execute(out Value,in Control,in Channel,in bInitial)"}.Value
   ) {is=struct RigVMUnitNode, template="SetControlVisibility::Execute(in Item,in Pattern,in bVisible)"}
   For_Each_4 = RigVMDispatchNode(Array=RigVMVariableNode() {variable=Robe Controls}.Value, ExecuteContext=@RigVMFunction_Sequence_3.D) {template="DISPATCH_RigVMDispatch_ArrayIterator(in Array,out Element,out Index,out Count,out Ratio)"}
   > &RigUnit_SetControlVisibility(
    Item=@For_Each_4.Element
    bVisible=&RigUnit_GetBoolAnimationChannel(Channel=Robe Controls Visibility, Control=control rig options, bInitial=false) {template="GetAnimationChannel::Execute(out Value,in Control,in Channel,in bInitial)"}.Value
   ) {template="SetControlVisibility::Execute(in Item,in Pattern,in bVisible)"}
   For_Each_1 = RigVMDispatchNode(Array=RigVMVariableNode() {variable=Doll Body Controls}.Value, ExecuteContext=@RigVMFunction_Sequence_3.C) {template="DISPATCH_RigVMDispatch_ArrayIterator(in Array,out Element,out Index,out Count,out Ratio)"}
   > &RigUnit_SetControlVisibility(
    Item=@For_Each_1.Element
    bVisible=&RigUnit_GetBoolAnimationChannel(Channel=Doll Body Controls Visibility, Control=control rig options, bInitial=false) {template="GetAnimationChannel::Execute(out Value,in Control,in Channel,in bInitial)"}.Value
   ) {template="SetControlVisibility::Execute(in Item,in Pattern,in bVisible)"}
   &RigUnit_SetControlVisibility(
    Item=(Type=Control,Name="dyn_doll_root")
    bVisible=&RigUnit_GetBoolAnimationChannel(Channel=Doll Root Visibility, Control=control rig options, bInitial=false) {template="GetAnimationChannel::Execute(out Value,in Control,in Channel,in bInitial)"}.Value
    ExecutePin=@RigVMFunction_Sequence_3.B
   ) {template="SetControlVisibility::Execute(in Item,in Pattern,in bVisible)"}
   For_Each_2 = RigVMDispatchNode(Array=RigVMVariableNode() {variable=Root Controls}.Value, ExecuteContext=@RigVMFunction_Sequence_3.A) {template="DISPATCH_RigVMDispatch_ArrayIterator(in Array,out Element,out Index,out Count,out Ratio)"}
   > &RigUnit_SetControlVisibility(
    Item=@For_Each_2.Element
    bVisible=&RigUnit_GetBoolAnimationChannel(Channel=Root Visibility, Control=control rig options, bInitial=false) {template="GetAnimationChannel::Execute(out Value,in Control,in Channel,in bInitial)"}.Value
   ) {template="SetControlVisibility::Execute(in Item,in Pattern,in bVisible)"}
   &RigUnit_SetControlVisibility(
    Item=(Type=Control,Name="dyn_door")
    bVisible=&RigUnit_GetBoolAnimationChannel(Channel=Door Control Visibility, Control=control rig options, bInitial=false) {template="GetAnimationChannel::Execute(out Value,in Control,in Channel,in bInitial)"}.Value
    ExecutePin=@RigVMFunction_Sequence_3.F
   ) {template="SetControlVisibility::Execute(in Item,in Pattern,in bVisible)"}
   RigVMFunctionReturnNode(ExecuteContext=@RigVMFunction_Sequence.B)
  ]}
 New Function_ContainedGraph {nodes [
   n1 = RigVMFunctionEntryNode()
   RigVMFunction_Sequence = &RigVMFunction_Sequence(ExecuteContext=@n1.ExecuteContext)
   RigVMDispatchNode(ExecuteContext=@RigVMFunction_Sequence.A, Array=RigVMVariableNode() {variable=ctrl array}.Value) {template="DISPATCH_RigVMDispatch_ArrayReset(io Array)"}
   > RigVMDispatchNode(Array=RigVMVariableNode() {variable=Channels array}.Value) {template="DISPATCH_RigVMDispatch_ArrayReset(io Array)"}
   > RigVMDispatchNode(Array=RigVMVariableNode() {variable=Root Controls}.Value) {template="DISPATCH_RigVMDispatch_ArrayReset(io Array)"}
   > RigVMDispatchNode(Array=RigVMVariableNode() {variable=Doll Body Controls}.Value) {template="DISPATCH_RigVMDispatch_ArrayReset(io Array)"}
   > RigVMDispatchNode(Array=RigVMVariableNode() {variable=Robe Controls}.Value) {template="DISPATCH_RigVMDispatch_ArrayReset(io Array)"}
   > RigVMDispatchNode(Array=RigVMVariableNode() {variable=Hair Controls}.Value) {template="DISPATCH_RigVMDispatch_ArrayReset(io Array)"}
   RigVMFunctionReturnNode(ExecuteContext=@RigVMFunction_Sequence.B)
  ]}
 RigVMFunctionLibrary {nodes [
   Animation Channels = RigVMCollapseNode(ExecuteContext=())
   Channel Hosting = RigVMCollapseNode(ExecuteContext=())
   Control Visibility = RigVMCollapseNode(ExecuteContext=())
   Create Controls = RigVMCollapseNode(Bone Array=(), ExecuteContext=())
   Reset Arrays = RigVMCollapseNode(ExecuteContext=())
   Set Bone Metadata = RigVMCollapseNode(
    Bone=(Type=Bone,Name="root")
    Color=(R=1,G=1,B=0,A=1)
    ExecuteContext=()
    Offset=(Rotation=(X=0,Y=0,Z=0,W=1),Translation=(X=0,Y=0,Z=0),Scale3D=(X=1,Y=1,Z=1))
    Parent=(Type=None,Name="None")
    Shape=Default
   )
  ]}
 RigVMFunction_Sequence_1_ContainedGraph {nodes [
   n1 = RigVMFunctionEntryNode()
   RigVMFunction_Sequence = &RigVMFunction_Sequence(ExecuteContext=@n1.ExecuteContext)
   n2 = &RigVMFunction_Sequence(ExecuteContext=@RigVMFunction_Sequence.B)
   n3 = &RigVMFunction_Sequence(ExecuteContext=@n2.B)
   n4 = &RigVMFunction_Sequence(ExecuteContext=@n3.B)
   RigVMFunctionReturnNode(A=@RigVMFunction_Sequence.A, B=@n2.A, C=@n3.A, D=@n4.A, F=@n4.B)
  ]}
 RigVMFunction_Sequence_2_ContainedGraph {nodes [
   n1 = RigVMFunctionEntryNode()
   n2 = &RigVMFunction_Sequence(ExecuteContext=@n1.ExecuteContext)
   n3 = &RigVMFunction_Sequence(ExecuteContext=@n2.B)
   n4 = &RigVMFunction_Sequence(ExecuteContext=@n3.B)
   RigVMFunctionReturnNode(A=@n2.A, B=@n3.A, C=@n4.A, D=@n4.B)
  ]}
 RigVMFunction_Sequence_2_ContainedGraph {nodes [
   n1 = RigVMFunctionEntryNode()
   n2 = &RigVMFunction_Sequence(ExecuteContext=@n1.ExecuteContext)
   n3 = &RigVMFunction_Sequence(ExecuteContext=@n2.B)
   n4 = &RigVMFunction_Sequence(ExecuteContext=@n3.B)
   n5 = &RigVMFunction_Sequence(ExecuteContext=@n4.B)
   n6 = &RigVMFunction_Sequence(ExecuteContext=@n5.B)
   RigVMFunctionReturnNode(A=@n2.A, B=@n3.A, C=@n4.A, D=@n5.A, E=@n6.A, F=@n6.B)
  ]}
 RigVMModel {nodes [
   RigUnit_BeginExecution = &RigUnit_BeginExecution()
   For_Each_1 = RigVMDispatchNode(ExecuteContext=@RigUnit_BeginExecution.ExecutePin, Array=RigVMVariableNode() {variable=bone array}.Value) {template="DISPATCH_RigVMDispatch_ArrayIterator(in Array,out Element,out Index,out Count,out Ratio)"}
   Get_Metadata_1 = RigVMDispatchNode(Default=(Type=Bone,Name="None"), Name=Control, NameSpace=Self, Item=@For_Each_1.Element) {template="DISPATCH_RigDispatch_GetMetadata(in Item,in Name,in NameSpace,in Default,out Value,out Found)"}
   > RigUnit_SetTransform = &RigUnit_SetTransform(
    Weight=1
    bInitial=False
    Item=@For_Each_1.Element
    Value=&RigUnit_GetTransform(bInitial=False, Item=@Get_Metadata_1.Value).Transform
   )
   Control Visibility = RigVMFunctionReferenceNode(ExecuteContext=@For_Each_1.Completed)
  ]}
 RigVMModel Construction Event {nodes [
   n1 = &RigUnit_PrepareForExecution()
   RigVMFunction_Sequence = RigVMAggregateNode(ExecuteContext=@n1.ExecutePin)
   For_Each_2 = RigVMDispatchNode(Array=RigVMVariableNode() {variable=bone array}.Value, ExecuteContext=@RigVMFunction_Sequence.B) {template="DISPATCH_RigVMDispatch_ArrayIterator(in Array,out Element,out Index,out Count,out Ratio)"}
   > Set Bone Metadata = RigVMFunctionReferenceNode(
    Bone=(Type=Bone,Name="root")
    Color=(R=1,G=1,B=0,A=1)
    Offset=(Rotation=(X=0,Y=0,Z=0,W=1),Translation=(X=0,Y=0,Z=0),Scale3D=(X=0.5,Y=0.5,Z=0.5))
    Parent=(Type=None,Name="None")
    Shape=Box_Thick
   )
   > Set Bone Metadata_1 = RigVMFunctionReferenceNode(
    Bone=(Type=Bone,Name="pack")
    Color=(R=1,G=1,B=0,A=1)
    Offset=(Rotation=(X=0,Y=0,Z=0,W=1),Translation=(X=0,Y=0.5,Z=0),Scale3D=(X=0.3,Y=0.3,Z=0.3))
    Parent=(Type=Bone,Name="root")
    Shape=Box_Thick
   )
   > Set Bone Metadata_2 = RigVMFunctionReferenceNode(
    Bone=(Type=Bone,Name="dyn_door")
    Color=(R=1,G=0.9136,B=0.02,A=0.5)
    Offset=(Rotation=(X=0,Y=0,Z=0,W=1),Translation=(X=10,Y=0,Z=0),Scale3D=(X=2.2,Y=0.2,Z=5))
    Parent=(Type=Bone,Name="pack")
    Shape=Box_Thin
   )
   > Set Bone Metadata_6 = RigVMFunctionReferenceNode(
    Bone=(Type=Bone,Name="dyn_doll_root")
    Color=(R=0.3864,G=0.1845,B=1,A=0.502)
    Offset=(Rotation=(X=0,Y=0,Z=0,W=1),Translation=(X=0,Y=0,Z=0),Scale3D=(X=2,Y=2,Z=1))
    Parent=(Type=None,Name="None")
    Shape=Arrow4_Thick
   )
   > Set Bone Metadata_17 = RigVMFunctionReferenceNode(
    Bone=(Type=Bone,Name="dyn_doll_pelvis")
    Color=(R=0.3864,G=0.1845,B=1,A=0.502)
    Offset=(Rotation=(X=0,Y=-0.7071,Z=0,W=0.7071),Translation=(X=0.5,Y=0,Z=0),Scale3D=(X=1,Y=1,Z=2))
    Parent=(Type=None,Name="None")
    Shape=Circle_Thick
   )
   > Set Bone Metadata_18 = RigVMFunctionReferenceNode(
    Bone=(Type=Bone,Name="dyn_doll_torso")
    Color=(R=0.3864,G=0.1845,B=1,A=0.502)
    Offset=(Rotation=(X=0,Y=-0.7071,Z=0,W=0.7071),Translation=(X=0,Y=0,Z=0),Scale3D=(X=1,Y=1,Z=2))
    Parent=(Type=None,Name="None")
    Shape=Circle_Thick
   )
   > Set Bone Metadata_19 = RigVMFunctionReferenceNode(
    Bone=(Type=Bone,Name="dyn_doll_head")
    Color=(R=0.3864,G=0.1845,B=1,A=0.502)
    Offset=(Rotation=(X=0,Y=0,Z=0,W=1),Translation=(X=7,Y=0,Z=0),Scale3D=(X=1.5,Y=1.5,Z=1.5))
    Parent=(Type=None,Name="None")
    Shape=Sphere_Thin
   )
   For_Each_1 = RigVMDispatchNode(Array=RigVMVariableNode() {variable=bone array}.Value, ExecuteContext=@RigVMFunction_Sequence.C) {template="DISPATCH_RigVMDispatch_ArrayIterator(in Array,out Element,out Index,out Count,out Ratio)"}
   Contains = &RigVMFunction_Contains(Search=arm_l, Name=@For_Each_1.Element.Name) {template="Contains::Execute(in Name,in Search,out Result)"}
   RigVMFunction_ControlFlowBranch = &RigVMFunction_ControlFlowBranch(Condition=@Contains.Result, ExecuteContext=@For_Each_1.ExecuteContext)
   n2 = &RigVMFunction_ControlFlowBranch(
    Condition=&RigVMFunction_Contains(Search=arm_r, Name=@For_Each_1.Element.Name) {template="Contains::Execute(in Name,in Search,out Result)"}.Result
    ExecuteContext=@RigVMFunction_ControlFlowBranch.Completed
   )
   Set Bone Metadata_9 = RigVMFunctionReferenceNode(
    Bone=(Type=Bone,Name="doll_thigh_l")
    Color=(R=0.016,G=0.7231,B=1,A=1)
    Offset=(Rotation=(X=0,Y=-0.7071,Z=0,W=0.7071),Translation=(X=0,Y=0,Z=0),Scale3D=(X=0.5,Y=0.5,Z=1))
    Parent=(Type=None,Name="None")
    Shape=Circle_Thick
    ExecuteContext=@n2.Completed
   )
   > Set Bone Metadata_12 = RigVMFunctionReferenceNode(
    Bone=(Type=Bone,Name="doll_calf_l")
    Color=(R=0.016,G=0.7231,B=1,A=1)
    Offset=(Rotation=(X=0,Y=-0.7071,Z=0,W=0.7071),Translation=(X=0,Y=0,Z=0),Scale3D=(X=0.5,Y=0.5,Z=1))
    Parent=(Type=None,Name="None")
    Shape=Circle_Thick
   )
   > Set Bone Metadata_14 = RigVMFunctionReferenceNode(
    Bone=(Type=Bone,Name="doll_foot_l")
    Color=(R=0.016,G=0.7231,B=1,A=1)
    Offset=(Rotation=(X=0,Y=-0.7071,Z=0,W=0.7071),Translation=(X=0,Y=1.5,Z=0),Scale3D=(X=0.3,Y=0.4,Z=0.2))
    Parent=(Type=None,Name="None")
    Shape=Box_Thick
   )
   > Set Bone Metadata_10 = RigVMFunctionReferenceNode(
    Bone=(Type=Bone,Name="doll_thigh_r")
    Color=(R=1,G=0.0176,B=0.0176,A=1)
    Offset=(Rotation=(X=0,Y=-0.7071,Z=0,W=0.7071),Translation=(X=0,Y=0,Z=0),Scale3D=(X=0.5,Y=0.5,Z=1))
    Parent=(Type=None,Name="None")
    Shape=Circle_Thick
   )
   > Set Bone Metadata_13 = RigVMFunctionReferenceNode(
    Bone=(Type=Bone,Name="doll_calf_r")
    Color=(R=1,G=0.0176,B=0.0176,A=1)
    Offset=(Rotation=(X=0,Y=-0.7071,Z=0,W=0.7071),Translation=(X=0,Y=0,Z=0),Scale3D=(X=0.5,Y=0.5,Z=1))
    Parent=(Type=None,Name="None")
    Shape=Circle_Thick
   )
   > Set Bone Metadata_15 = RigVMFunctionReferenceNode(
    Bone=(Type=Bone,Name="doll_foot_r")
    Color=(R=1,G=0.0176,B=0.0176,A=1)
    Offset=(Rotation=(X=0,Y=-0.7071,Z=0,W=0.7071),Translation=(X=0,Y=-1.5,Z=0),Scale3D=(X=0.3,Y=0.4,Z=0.2))
    Parent=(Type=None,Name="None")
    Shape=Box_Thick
   )
   Set Bone Metadata_11 = RigVMFunctionReferenceNode(
    Color=(R=1,G=0.0176,B=0.0176,A=1)
    Offset=(Rotation=(X=0,Y=-0.7071,Z=0,W=0.7071),Translation=(X=0,Y=0,Z=0),Scale3D=(X=0.5,Y=0.5,Z=1))
    Parent=(Type=None,Name="None")
    Shape=Circle_Thick
    ExecuteContext=@n2.True
    Bone=@For_Each_1.Element
   )
   Set Bone Metadata_3 = RigVMFunctionReferenceNode(
    Color=(R=0.016,G=0.7231,B=1,A=1)
    Offset=(Rotation=(X=0,Y=-0.7071,Z=0,W=0.7071),Translation=(X=0,Y=0,Z=0),Scale3D=(X=0.5,Y=0.5,Z=1))
    Parent=(Type=None,Name="None")
    Shape=Circle_Thick
    ExecuteContext=@RigVMFunction_ControlFlowBranch.True
    Bone=@For_Each_1.Element
   )
   For_Each_3 = RigVMDispatchNode(Array=RigVMVariableNode() {variable=bone array}.Value, ExecuteContext=@RigVMFunction_Sequence.D) {template="DISPATCH_RigVMDispatch_ArrayIterator(in Array,out Element,out Index,out Count,out Ratio)"}
   n3 = &RigVMFunction_ControlFlowBranch(
    Condition=&RigVMFunction_Contains(Search=robe, Name=@For_Each_3.Element.Name) {template="Contains::Execute(in Name,in Search,out Result)"}.Result
    ExecuteContext=@For_Each_3.ExecuteContext
   )
   n4 = &RigVMFunction_ControlFlowBranch(
    Condition=&RigVMFunction_Contains(Search=hair_bck, Name=@For_Each_3.Element.Name) {template="Contains::Execute(in Name,in Search,out Result)"}.Result
    ExecuteContext=@n3.Completed
   )
   n5 = &RigVMFunction_ControlFlowBranch(
    Condition=&RigVMFunction_Contains(Search=hair_fr, Name=@For_Each_3.Element.Name) {template="Contains::Execute(in Name,in Search,out Result)"}.Result
    ExecuteContext=@n4.Completed
   )
   Set Bone Metadata_8 = RigVMFunctionReferenceNode(
    Color=(R=0.048,G=0.73,B=0,A=1)
    Offset=(Rotation=(X=0,Y=-0.7071,Z=0,W=0.7071),Translation=(X=0,Y=0,Z=0),Scale3D=(X=0.2,Y=0.2,Z=0.3))
    Parent=(Type=None,Name="None")
    Shape=Box_Thick
    ExecuteContext=@n5.True
    Bone=@For_Each_3.Element
   )
   > Set Bone Metadata_16 = RigVMFunctionReferenceNode(
    Bone=(Type=Bone,Name="dyn_doll_ribbon")
    Color=(R=0.9911,G=0.6172,B=1,A=1)
    Offset=(Rotation=(X=0.0432,Y=-0.9905,Z=0.0057,W=0.1304),Translation=(X=4.5,Y=0.7,Z=-1.1),Scale3D=(X=0.5,Y=1.4,Z=0.7))
    Parent=(Type=None,Name="None")
    Shape=Wedge_Thin
   )
   Set Bone Metadata_5 = RigVMFunctionReferenceNode(
    Color=(R=0.0482,G=0.7305,B=0,A=1)
    Offset=(Rotation=(X=0,Y=-0.7071,Z=0,W=0.7071),Translation=(X=0,Y=0,Z=0),Scale3D=(X=0.2,Y=0.5,Z=0.5))
    Parent=(Type=None,Name="None")
    Shape=Box_Thick
    ExecuteContext=@n4.True
    Bone=@For_Each_3.Element
   )
   Set Bone Metadata_4 = RigVMFunctionReferenceNode(
    Color=(R=1,G=0.3325,B=0.0222,A=1)
    Offset=(Rotation=(X=0,Y=-0.7071,Z=0,W=0.7071),Translation=(X=0,Y=0,Z=0),Scale3D=(X=0.2,Y=0.5,Z=0.5))
    Parent=(Type=None,Name="None")
    Shape=Box_Thick
    ExecuteContext=@n3.True
    Bone=@For_Each_3.Element
   )
   Create Controls = RigVMFunctionReferenceNode(Bone Array=RigVMVariableNode() {variable=bone array}.Value, ExecuteContext=@RigVMFunction_Sequence.F)
   > Channel Hosting = RigVMFunctionReferenceNode()
   n6 = &RigUnit_HierarchyImportFromSkeleton(NameSpace=None, bIncludeCurves=False, bIncludeMeshSockets=False, bIncludeVirtualBones=True, ExecutePin=@RigVMFunction_Sequence.A)
   > RigVMVariableNode(Value=@n6.Items) {variable=bone array}
   > Reset Arrays = RigVMFunctionReferenceNode()
   > Animation Channels = RigVMFunctionReferenceNode()
  ]}
 RigVMModel Rig {nodes [
   n1 = /Script/ControlRig.RigUnit_InverseExecution() {is=struct RigVMUnitNode}
   For_Each_1 = RigVMDispatchNode(ExecuteContext=@n1.ExecutePin, Array=RigVMVariableNode() {variable=bone array}.Value) {template="DISPATCH_RigVMDispatch_ArrayIterator(in Array,out Element,out Index,out Count,out Ratio)"}
   Get_Metadata_1 = RigVMDispatchNode(Default=(Type=Bone,Name="None"), Name=Control, NameSpace=Self, Item=@For_Each_1.Element) {template="DISPATCH_RigDispatch_GetMetadata(in Item,in Name,in NameSpace,in Default,out Value,out Found)"}
   > RigUnit_SetTransform = &RigUnit_SetTransform(
    Weight=1
    bInitial=False
    Item=@For_Each_1.Element
    Value=&RigUnit_GetTransform(bInitial=False, Item=@Get_Metadata_1.Value).Transform
   )
  ]}
]
hierarchy [control rig options {controlType=EULER_TRANSFORM, type=Control, loc=<0 0 0>, rot=<0 0 0 1>}]
variables [
 Channels array {type="TArray<FRigElementKey>"}
 Doll Body Controls {type="TArray<FRigElementKey>"}
 Hair Controls {type="TArray<FRigElementKey>"}
 Robe Controls {type="TArray<FRigElementKey>"}
 Root Controls {type="TArray<FRigElementKey>"}
 bone array {type="TArray<FRigElementKey>"}
 ctrl array {type="TArray<FRigElementKey>"}
]
