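# Physics asset for the doll on the NeonFriction backpack cosmetic, set up for secondary motion of its limbs, head and hair.
# Capsules cover the doll's bones, while the pack body adds three boxes that approximate its bulk.
# Angularly locked constraints chain root, pack, doll root, pelvis and torso into a rigid spine.
# Thighs, upper arms and head hang off it with limited swing, and the two front hair bones rotate freely on the head.
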PhysicsAsset /BRCosmetics/Accessories/FORT_Backpacks/Backpack_M_MED_NeonFriction/Meshes/M_MED_NeonFriction_Pack_DollLimbs_Physics
bodies {
 doll_thigh_l [{shapeType=Capsule, length=6, radius=2.5, shapeName=doll_thigh_l_capsule, center=<-4.5432 0 -0.9084>, rotation=<rot -90 0.0002 0>}]
 doll_thigh_r [{shapeType=Capsule, length=6, radius=2.5, shapeName=doll_thigh_r_capsule, center=<4.6297 0 0.7306>, rotation=<rot -90 0.0002 0>}]
 doll_upperarm_l [{shapeType=Capsule, length=3, radius=2, shapeName=doll_upperarm_l_capsule, center=<3.8 0 0>, rotation=<rot -90 0.0001 0>}]
 doll_upperarm_r [{shapeType=Capsule, length=3, radius=2, shapeName=doll_upperarm_r_capsule, center=<-3.8 0 0>, rotation=<rot -90 0.0001 0>}]
 dyn_doll_hair_fr_l_base [{
   shapeType=Capsule
   length=7.1925
   radius=1.2872
   shapeName=dyn_doll_hair_fr_l_base_capsule
   center=<2.1076 0.2234 1.3566>
   rotation=<rot 0 -93.9506 -99.468>
  }]
 dyn_doll_hair_fr_r_base [{
   shapeType=Capsule
   length=7.2469
   radius=1.1786
   shapeName=dyn_doll_hair_fr_r_base_capsule
   center=<3.9806 -0.0707 0.5238>
   rotation=<rot 0 -88.9002 -83.524>
  }]
 dyn_doll_head [{shapeType=Capsule, length=1.5, radius=7.5, shapeName=dyn_doll_head_capsule, center=<6.4776 0 0>, rotation=<rot 90 0.0004 0>}]
 dyn_doll_pelvis [{shapeType=Capsule, length=3.5, radius=3, shapeName=dyn_doll_pelvis_capsule, center=<-0.5544 0 0>}]
 dyn_doll_root [{shapeType=Capsule, length=0, radius=2, shapeName=dyn_doll_root_capsule, rotation=<rot 90 -0.0002 0>}]
 dyn_doll_torso [{shapeType=Capsule, length=2, radius=3, shapeName=dyn_doll_torso_capsule, center=<0.7342 0 0>}]
 pack [
  {shapeType=Capsule, length=4, radius=2, shapeName=pack_capsule}
  {shapeType=Box, extentX=21.5, extentY=7.5, extentZ=42.5, shapeName=pack_box, center=<0 2.1785 -6.7278>}
  {shapeType=Box, extentX=7.5, extentY=22.5, extentZ=24.5, shapeName=pack_box, center=<14.2 -11.0144 -15>}
  {shapeType=Box, extentX=7.5, extentY=22.5, extentZ=24.5, shapeName=pack_box, center=<-14.2 -11.0144 -15>}
 ]
 root [{shapeType=Capsule, length=4, radius=2, shapeName=root_capsule}]
}
constraints [
 {
  bone1Name=pack
  bone2Name=root
  swing1LimitDegrees=45
  swing1Motion=Locked
  swing2LimitDegrees=45
  swing2Motion=Locked
  twistLimitDegrees=45
  twistMotion=Locked
 }
 {
  bone1Name=dyn_doll_root
  bone2Name=pack
  swing1LimitDegrees=45
  swing1Motion=Locked
  swing2LimitDegrees=45
  swing2Motion=Locked
  twistLimitDegrees=45
  twistMotion=Locked
 }
 {
  bone1Name=dyn_doll_pelvis
  bone2Name=dyn_doll_root
  swing1LimitDegrees=45
  swing1Motion=Locked
  swing2LimitDegrees=45
  swing2Motion=Locked
  twistLimitDegrees=45
  twistMotion=Locked
 }
 {bone1Name=doll_thigh_l, bone2Name=dyn_doll_pelvis, swing1LimitDegrees=45, swing2LimitDegrees=45, twistLimitDegrees=0}
 {bone1Name=doll_thigh_r, bone2Name=dyn_doll_pelvis, swing1LimitDegrees=45, swing2LimitDegrees=45, twistLimitDegrees=0}
 {
  bone1Name=dyn_doll_torso
  bone2Name=dyn_doll_pelvis
  swing1LimitDegrees=45
  swing1Motion=Locked
  swing2LimitDegrees=45
  swing2Motion=Locked
  twistLimitDegrees=45
  twistMotion=Locked
 }
 {bone1Name=doll_upperarm_l, bone2Name=dyn_doll_torso, swing1LimitDegrees=20, swing2LimitDegrees=45, twistLimitDegrees=0}
 {bone1Name=doll_upperarm_r, bone2Name=dyn_doll_torso, swing1LimitDegrees=20, swing2LimitDegrees=45, twistLimitDegrees=0}
 {bone1Name=dyn_doll_head, bone2Name=dyn_doll_torso, swing1LimitDegrees=45, swing2LimitDegrees=45, twistLimitDegrees=15}
 {
  bone1Name=dyn_doll_hair_fr_l_base
  bone2Name=dyn_doll_head
  swing1LimitDegrees=45
  swing1Motion=Free
  swing2LimitDegrees=45
  swing2Motion=Free
  twistLimitDegrees=45
  twistMotion=Free
 }
 {
  bone1Name=dyn_doll_hair_fr_r_base
  bone2Name=dyn_doll_head
  swing1LimitDegrees=45
  swing1Motion=Free
  swing2LimitDegrees=45
  swing2Motion=Free
  twistLimitDegrees=45
  twistMotion=Free
 }
]
