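# Physics asset for the NeonFriction cosmetic backpack, giving its attached doll's hair and limbs physics bodies for secondary motion.
# Capsules cover the doll, while boxes add the pack's bulk as collision shapes.
# Locked angular joints hold root, pack and the doll's pelvis, torso and head rigid.
# Unlocked joints carry the arms, thighs, two front hair chains and a back chain that forks left and right.
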PhysicsAsset /BRCosmetics/Accessories/FORT_Backpacks/Backpack_M_MED_NeonFriction/Meshes/M_MED_NeonFriction_Pack_Hair_Physics
bodies {
 doll_thigh_l [{shapeType=Capsule, length=6, radius=2.5, shapeName=doll_thigh_l_capsule, center=<-4.5432 0 -0.9084>, rotation=<rot -90 0.0002 0>}]
 doll_thigh_r [{shapeType=Capsule, length=6, radius=2.5, shapeName=doll_thigh_r_capsule, center=<4.6297 0 0.7306>, rotation=<rot -90 0.0002 0>}]
 doll_upperarm_l [{shapeType=Capsule, length=4, radius=2, shapeName=doll_upperarm_l_capsule, center=<3.8 0 0>, rotation=<rot -90 0.0001 0>}]
 doll_upperarm_r [{shapeType=Capsule, length=4, radius=2, shapeName=doll_upperarm_r_capsule, center=<-3.8 0 0>, rotation=<rot -90 0.0001 0>}]
 dyn_doll_hair_bck_base [{shapeType=Capsule, length=2.5, radius=3.5, shapeName=dyn_doll_hair_bck_base_capsule, center=<2.7503 0 0>, rotation=<rot 15 90.0003 90.001>}]
 dyn_doll_hair_bck_l_02 [{
   shapeType=Capsule
   length=5.5
   radius=3
   shapeName=dyn_doll_hair_bck_l_02_capsule
   center=<1.8475 -0.3147 -0.0843>
   rotation=<rot -2.5753 -9.6661 80.2179>
  }]
 dyn_doll_hair_bck_l_03 [{
   shapeType=Capsule
   length=4.5
   radius=3
   shapeName=dyn_doll_hair_bck_l_03_capsule
   center=<1.7134 -1.0667 -0.2159>
   rotation=<rot -2.5754 -9.6662 80.218>
  }]
 dyn_doll_hair_bck_mid_01 [{
   shapeType=Capsule
   length=2.5
   radius=3.5
   shapeName=dyn_doll_hair_bck_mid_01_capsule
   center=<2.7503 0 0>
   rotation=<rot 15 90.0002 90.0007>
  }]
 dyn_doll_hair_bck_r_02 [{
   shapeType=Capsule
   length=4.5
   radius=3
   shapeName=dyn_doll_hair_bck_r_02_capsule
   center=<1.8688 0.1579 0.0423>
   rotation=<rot 1.2931 4.8306 70.0544>
  }]
 dyn_doll_hair_bck_r_03 [{shapeType=Capsule, length=4.5, radius=3, shapeName=dyn_doll_hair_bck_r_03_capsule, center=<1.876 0 0>, rotation=<rot 0.0005 0.0002 69.9999>}]
 dyn_doll_hair_fr_l_01 [{
   shapeType=Capsule
   length=4.5
   radius=1.5
   shapeName=dyn_doll_hair_fr_l_01_capsule
   center=<2.0135 -0.2441 0.9109>
   rotation=<rot 15 89.9996 89.9991>
  }]
 dyn_doll_hair_fr_l_02 [{
   shapeType=Capsule
   length=4.5
   radius=1.5
   shapeName=dyn_doll_hair_fr_l_02_capsule
   center=<2.0135 -0.2441 0.9109>
   rotation=<rot 15 89.9995 89.9991>
  }]
 dyn_doll_hair_fr_l_base [{shapeType=Capsule, length=4, radius=2, shapeName=dyn_doll_hair_fr_l_base_capsule, rotation=<rot 15 89.9997 89.9991>}]
 dyn_doll_hair_fr_r_01 [{
   shapeType=Capsule
   length=4.5
   radius=1.5
   shapeName=dyn_doll_hair_fr_r_01_capsule
   center=<2.0135 -0.2441 0.9109>
   rotation=<rot 15 89.9998 89.9991>
  }]
 dyn_doll_hair_fr_r_02 [{
   shapeType=Capsule
   length=4.5
   radius=1.5
   shapeName=dyn_doll_hair_fr_r_02_capsule
   center=<2.0135 -0.2441 0.9109>
   rotation=<rot 15.0001 90.0004 89.9994>
  }]
 dyn_doll_hair_fr_r_base [{shapeType=Capsule, length=4, radius=2, shapeName=dyn_doll_hair_fr_r_base_capsule, rotation=<rot 15 89.9998 89.9991>}]
 dyn_doll_head [{shapeType=Capsule, length=3, radius=3, shapeName=dyn_doll_head_capsule, center=<4.3261 0 0>, rotation=<rot 90 0.0005 0>}]
 dyn_doll_pelvis [{shapeType=Capsule, length=3.5, radius=3, shapeName=dyn_doll_pelvis_capsule, center=<-0.5544 0 0>}]
 dyn_doll_root [{shapeType=Capsule, length=0, radius=2, shapeName=dyn_doll_root_capsule, rotation=<rot 90 -0.0001 0>}]
 dyn_doll_torso [{shapeType=Capsule, length=2, radius=3, shapeName=dyn_doll_torso_capsule, center=<0.7342 0 0>}]
 pack [
  {shapeType=Capsule, length=4, radius=2, shapeName=pack_capsule}
  {shapeType=Box, extentX=21.5, extentY=7.5, extentZ=42.5, shapeName=pack_box, center=<0 2.1785 -6.7278>}
  {shapeType=Box, extentX=7.5, extentY=22.5, extentZ=24.5, shapeName=pack_box, center=<14.2 -11.0144 -15>}
  {shapeType=Box, extentX=7.5, extentY=22.5, extentZ=24.5, shapeName=pack_box, center=<-14.2 -11.0144 -15>}
 ]
 root [{shapeType=Capsule, length=4, radius=2, shapeName=root_capsule}]
}
constraints [
 {
  bone1Name=pack
  bone2Name=root
  swing1LimitDegrees=45
  swing1Motion=Locked
  swing2LimitDegrees=45
  swing2Motion=Locked
  twistLimitDegrees=45
  twistMotion=Locked
 }
 {
  bone1Name=dyn_doll_root
  bone2Name=pack
  swing1LimitDegrees=45
  swing1Motion=Locked
  swing2LimitDegrees=45
  swing2Motion=Locked
  twistLimitDegrees=45
  twistMotion=Locked
 }
 {
  bone1Name=dyn_doll_pelvis
  bone2Name=dyn_doll_root
  swing1LimitDegrees=45
  swing1Motion=Locked
  swing2LimitDegrees=45
  swing2Motion=Locked
  twistLimitDegrees=45
  twistMotion=Locked
 }
 {bone1Name=doll_thigh_l, bone2Name=dyn_doll_pelvis, swing1LimitDegrees=45, swing2LimitDegrees=45, twistLimitDegrees=0}
 {bone1Name=doll_thigh_r, bone2Name=dyn_doll_pelvis, swing1LimitDegrees=45, swing2LimitDegrees=45, twistLimitDegrees=0}
 {
  bone1Name=dyn_doll_torso
  bone2Name=dyn_doll_pelvis
  swing1LimitDegrees=45
  swing1Motion=Locked
  swing2LimitDegrees=45
  swing2Motion=Locked
  twistLimitDegrees=45
  twistMotion=Locked
 }
 {bone1Name=doll_upperarm_l, bone2Name=dyn_doll_torso, swing1LimitDegrees=45, swing2LimitDegrees=45, twistLimitDegrees=0}
 {bone1Name=doll_upperarm_r, bone2Name=dyn_doll_torso, swing1LimitDegrees=45, swing2LimitDegrees=45, twistLimitDegrees=0}
 {
  bone1Name=dyn_doll_head
  bone2Name=dyn_doll_torso
  swing1LimitDegrees=45
  swing1Motion=Locked
  swing2LimitDegrees=45
  swing2Motion=Locked
  twistLimitDegrees=45
  twistMotion=Locked
 }
 {bone1Name=dyn_doll_hair_bck_base, bone2Name=dyn_doll_head, swing1LimitDegrees=45, swing2LimitDegrees=45, twistLimitDegrees=45}
 {bone1Name=dyn_doll_hair_bck_mid_01, bone2Name=dyn_doll_hair_bck_base, swing1LimitDegrees=45, swing2LimitDegrees=45, twistLimitDegrees=0}
 {bone1Name=dyn_doll_hair_bck_l_02, bone2Name=dyn_doll_hair_bck_mid_01, swing1LimitDegrees=45, swing2LimitDegrees=45, twistLimitDegrees=0}
 {bone1Name=dyn_doll_hair_bck_l_03, bone2Name=dyn_doll_hair_bck_l_02, swing1LimitDegrees=45, swing2LimitDegrees=45, twistLimitDegrees=0}
 {bone1Name=dyn_doll_hair_bck_r_02, bone2Name=dyn_doll_hair_bck_mid_01, swing1LimitDegrees=45, swing2LimitDegrees=45, twistLimitDegrees=0}
 {bone1Name=dyn_doll_hair_bck_r_03, bone2Name=dyn_doll_hair_bck_r_02, swing1LimitDegrees=45, swing2LimitDegrees=45, twistLimitDegrees=0}
 {bone1Name=dyn_doll_hair_fr_l_base, bone2Name=dyn_doll_head, swing1LimitDegrees=45, swing2LimitDegrees=45, twistLimitDegrees=0}
 {bone1Name=dyn_doll_hair_fr_l_01, bone2Name=dyn_doll_hair_fr_l_base, swing1LimitDegrees=20, swing2LimitDegrees=45, twistLimitDegrees=0}
 {bone1Name=dyn_doll_hair_fr_l_02, bone2Name=dyn_doll_hair_fr_l_01, swing1LimitDegrees=20, swing2LimitDegrees=45, twistLimitDegrees=0}
 {bone1Name=dyn_doll_hair_fr_r_base, bone2Name=dyn_doll_head, swing1LimitDegrees=20, swing2LimitDegrees=45, twistLimitDegrees=0}
 {bone1Name=dyn_doll_hair_fr_r_01, bone2Name=dyn_doll_hair_fr_r_base, swing1LimitDegrees=20, swing2LimitDegrees=45, twistLimitDegrees=0}
 {bone1Name=dyn_doll_hair_fr_r_02, bone2Name=dyn_doll_hair_fr_r_01, swing1LimitDegrees=20, swing2LimitDegrees=45, twistLimitDegrees=0}
]
