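# Animation Blueprint for the NeonFriction backpack, switching between scripted reactive animation and daytime physics.
# At night or with frontend reactivity, a state machine plays Idle and React sequences; otherwise a cached Control Rig pose feeds doll limb and hair RigidBody nodes, then ribbon AnimDynamics.
# The update event derives that switch from day phase and starts the first reaction after a delay.
# The ResetReactivity notify schedules repeats, randomly timed outside the frontend.
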AnimBlueprint /BRCosmetics/Accessories/FORT_Backpacks/Backpack_M_MED_NeonFriction/Meshes/M_MED_NeonFriction_Pack_AnimBP
target_skeleton=&/M_MED_NeonFriction_Pack_Skeleton
graphs [
 AnimGraph {nodes [
   Component To Local(ComponentPose="Transform (Modify) Bone - Bone: dyn_door"(ComponentPose=Local To Component(), Alpha=1, Translation="0,0,0", Rotation="-0.000,-140.000,0.000", Scale="1,1,1") {rotationMode=BMM_Additive, translationMode=BMM_Additive, boneToModify {boneName=dyn_door}, rotation {yaw=-140}})
   n1 = RigidBody!(ComponentPose=Local To Component(LocalPose=Use cached pose 'SetupCache'()), AlphaCurveName:Name=BodyDynamicsAlpha, ExternalForce) -> (Pose) {
    alphaInputType=Curve
    bOverrideWorldGravity=true
    overridePhysicsAsset=&/M_MED_NeonFriction_Pack_DollLimbs_Physics
    simulationSpace=BaseBoneSpace
    alphaScaleBiasClamp {bias=1, scale=-1}
    baseBoneRef {boneName=root}
    componentAppliedLinearAccClamp=<3000 3000 3000>
    overrideWorldGravity {z=-980}
    simSpaceSettings {
     dampingAlpha=0
     maxAngularAcceleration=40
     maxAngularVelocity=8
     maxLinearAcceleration=4000
     maxLinearVelocity=80
     velocityScaleZ=0.4
     worldAlpha=1
     externalLinearDragV=<0.2 0.2 0.2>
    }
   }
   n2 = RigidBody(ComponentPose=@n1, AlphaCurveName=HairDynamicsAlpha) {
    alphaInputType=Curve
    bOverrideWorldGravity=true
    overridePhysicsAsset=&/M_MED_NeonFriction_Pack_Hair_Physics
    simulationSpace=BaseBoneSpace
    alphaScaleBiasClamp {bias=1, scale=-1}
    baseBoneRef {boneName=root}
    componentAppliedLinearAccClamp=<3000 3000 3000>
    overrideWorldGravity {z=-980}
    simSpaceSettings {
     dampingAlpha=0
     maxAngularAcceleration=40
     maxAngularVelocity=8
     maxLinearAcceleration=4000
     maxLinearVelocity=80
     velocityScaleZ=0.5
     worldAlpha=1
     externalLinearDragV=<0.2 0.2 0.2>
    }
   }
   n3 = Anim Dynamics - Bone: dyn_doll_ribbon!(ComponentPose=@n2, AlphaCurveName:Name=HairDynamicsAlpha, ExternalForce="0,0,0") -> (Pose) {
    alphaInputType=Curve
    angularBiasOverride=0.99
    angularDampingOverride=1
    angularSpringConstant=5000
    bAngularSpring=true
    bGravityOverrideInSimSpace=true
    bOverrideAngularDamping=true
    bUseGravityOverride=true
    simulationSpace=RootRelative
    alphaScaleBiasClamp {bias=1, scale=-1}
    boundBone {boneName=dyn_doll_ribbon}
    componentAppliedLinearAccClamp=<1000 1000 1000>
    gravityOverride {z=-980}
    physicsBodyDefinitions [{
      boundBone {boneName=dyn_doll_ribbon}
      boxExtents=<3 3 3>
      constraintSetup {angularLimitsMax=<0 20 0>, angularLimitsMin=<0 -20 0>, angularTarget=<0 0 0>}
      localJointOffset=<-0.3 0 0>
     }]
    simSpaceSettings {maxAngularAcceleration=8, maxAngularVelocity=800, simSpaceAngularAlpha=1}
   }
   AnimGraphNode_StateMachine_0 = Reactive() {is=title AnimGraphNode_StateMachine}
   Output Pose(Result=Blend Poses by bool(
     BlendPose_0=@AnimGraphNode_StateMachine_0
     BlendPose_1=Component To Local(ComponentPose=@n3)
     bActiveValue=Get IsNightORFrontendReactive() -> (IsNightORFrontendReactive:bool)
    )) {layerGroup=None}
   n4 = Control Rig(Source=Input Pose()) {controlRigAssetReference {blueprintRigClass=&/M_MED_NeonFriction_Pack_CtrlRig.M_MED_NeonFriction_Pack_CtrlRig_C}}
   Save cached pose 'SetupCache'(Pose=@n4) {cacheName=SetupCache}
  ]}
 Reactive {
  entry=Idle
  path=AnimGraph.AnimGraphNode_StateMachine_0.Reactive
  ids=id
  nodes [
   AnimStateNode_0 = Idle()
   AnimStateNode_1 = React() {is=title AnimStateNode}
   AnimStateTransitionNode_0 = AnimStateTransitionNode() {bAllowInertializationForSelfTransitions=true, crossfadeDuration=0}
   AnimStateTransitionNode_1 = AnimStateTransitionNode() {bAllowInertializationForSelfTransitions=true, bAutomaticRuleBasedOnSequencePlayerInState=true, crossfadeDuration=0}
  ]
  links [
   {
    dst=React
    rule=AnimGraph.AnimGraphNode_StateMachine_0.Reactive.AnimStateTransitionNode_0.Transition
    src=Idle
    via=AnimStateTransitionNode_0
   }
   {
    dst=Idle
    rule=AnimGraph.AnimGraphNode_StateMachine_0.Reactive.AnimStateTransitionNode_1.Transition
    src=React
    via=AnimStateTransitionNode_1
   }
  ]
 }
 Idle {
  path=AnimGraph.AnimGraphNode_StateMachine_0.Reactive.AnimStateNode_0.Idle
  nodes [
   n1 = Sequence Player 'M_MED_NeonFriction_Idle_PACK'() {bLoopAnimation=false, sequence=/BRCosmetics/Accessories/FORT_Backpacks/Backpack_M_MED_NeonFriction/Anim/M_MED_NeonFriction_Idle_PACK}
   Output Animation Pose(Result=@n1)
  ]
 }
 React {
  path=AnimGraph.AnimGraphNode_StateMachine_0.Reactive.AnimStateNode_1.React
  nodes [
   n1 = Sequence Player 'M_MED_NeonFriction_Reactive_PACK'() {bLoopAnimation=false, sequence=/BRCosmetics/Accessories/FORT_Backpacks/Backpack_M_MED_NeonFriction/Anim/M_MED_NeonFriction_Reactive_PACK}
   Output Animation Pose(Result=@n1)
  ]
 }
 Transition {
  path=AnimGraph.AnimGraphNode_StateMachine_0.Reactive.AnimStateTransitionNode_0.Transition
  nodes [Result(bCanEnterTransition=Get PlayReactive() -> (PlayReactive:bool))]
 }
 Transition {path=AnimGraph.AnimGraphNode_StateMachine_0.Reactive.AnimStateTransitionNode_1.Transition, nodes [Result()]}
 EventGraph {nodes [
   Event BlueprintUpdateAnimation()
   n1 = GetCurrentDayPhase!() -> (ReturnValue:EFortDayPhase Enum=Morning)
   > n2 = Set isNight(isNight:bool="integer > integer"(A="To Integer (Byte)"!(InByte:Byte=@n1) -> (ReturnValue:int=0), B=2)) -> (Output_Get:bool)
   > n3 = Set IsNightORFrontendReactive(IsNightORFrontendReactive:bool=OR Boolean(A=@n2, B=Get bShouldDisplayFrontendReactivity() -> (bShouldDisplayFrontendReactivity:bool))) -> (Output_Get:bool)
   > Branch(Condition=@n3)
   > Do Once(Reset=@n5.then)
   > n4 = Delay(Duration=4)
   Set PlayReactive(execute=[@n7.then, @n8.then, @n4.then], PlayReactive:bool=true) -> (Output_Get:bool)
   n5 = Event OnDayPhaseChanged!() -> (OutputDelegate, then, CurrentDayPhase:EFortDayPhase Enum=Morning, PreviousDayPhase:EFortDayPhase Enum=Morning, bAtCreation:bool) {is=title K2Node_Event}
   AnimNotify_ResetReactivity() {is=title K2Node_Event}
   > Set PlayReactive(PlayReactive=false)
   > n6 = Branch(Condition=Get bShouldDisplayFrontendReactivity())
   > n7 = Delay(Duration=10)
   n8 = Delay(execute=@n6.else, Duration=RandomFloatInRange(Min=10, Max=30))
   TryGetPawnOwner()
   Event OnCharacterCustomizationLoaded()
  ]}
]
variables [
 isNight {type=boolean}
 IsNightORFrontendReactive {type=boolean}
 PlayReactive {type=boolean}
]
